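# Nanite-enabled lit material for background mountains in SparksMaps, built on Homebase rock, grass and soil textures.
# Switches enable world-aligned UVs, parameter-collection tints, a slope and vertex-red blend layer, and a soil layer masked by texture and vertex blue.
# Optional snow caps combine object-relative height with slope, then a BlueTint lerp colors the result.
# A base normal packed in UV1 and UV2 is angle-correct blended with an optional detail normal.
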Material /SparksMaps/Environment/Mountains/M_Mountains_BG_Sparks_A
graph {
 properties {bUsedWithNanite=true, numCustomizedUVs=3, pixelDepthOffsetMode=PDOM_Legacy, usedShadingModels=Default Lit}
 nodes [
  Constant() {r=0.2}
  n1 = CollectionParameter() {collection=/Game/Environments/Homebase/Materials/HomeBaseParameterCollection, parameterName=GrassTextureSize}
  n2 = Divide(
   A=ComponentMask(None=WorldPosition() {worldPositionShaderOffset=WPT_ExcludeAllShaderOffsets}.XYZ) {g=true, r=true}
   B=StaticSwitchParameter(True=ComponentMask(None=@n1) {r=true}, False=ScalarParameter() {defaultValue=512, group=UVmapping, parameterName=UVTextureSize}) {group=Options, parameterName=UseGrassTextureSize}
  )
  n3 = StaticSwitchParameter(True=@n2, False=TextureCoordinate()) {group=Options, parameterName=DiffuseWorldAlignedUV}
  n4 = CollectionParameter() {collection=&HomeBaseParameterCollection, parameterName=RockColor}
  n5 = LinearInterpolate(
   A=ComponentMask(None=CollectionParameter() {collection=&HomeBaseParameterCollection, parameterName=GrassColorDead}) {b=true, g=true, r=true}
   B=ComponentMask(None=CollectionParameter() {collection=&HomeBaseParameterCollection, parameterName=GrassColor}) {b=true, g=true, r=true}
   Alpha=VertexColor().G
  )
  n6 = VectorParameter() {group=Base, parameterName=DiffuseColor, defaultValue=<rgba 1 1 1 1>}
  n7 = Multiply(
   A=TextureSampleParameter2D(UVs=@n3) {parameterName=Diffuse, texture=/Game/Environments/Homebase/Textures/T_HB_Rock_Scattered_D}.RGB
   B=StaticSwitchParameter(
    True=ComponentMask(None=@n4) {b=true, g=true, r=true}
    False=StaticSwitchParameter(True=@n5, False=@n6.RGB) {group=Options, parameterName=UseGrassColorforDiffuse}
   ) {group=Options, parameterName=UseRockColorforDiffsue}
  )
  n8 = StaticSwitchParameter(
   True=ComponentMask(None=CollectionParameter() {collection=&HomeBaseParameterCollection, parameterName=SoilColor}) {b=true, g=true, r=true}
   False=VectorParameter() {group=Blend2, parameterName=BlendColor2, defaultValue=<rgba 0.78 0.78 0.78 1>}.RGB
  ) {group=Options, parameterName=UseSoilColorforBlendTexture2}
  n9 = StaticSwitchParameter(True=@n2, False=TextureCoordinate()) {group=Options, parameterName=BlendWorldAlignedUV}
  n10 = Multiply(A=ScalarParameter() {defaultValue=1, parameterName=Blend2Scale}, B=@n9)
  n11 = Multiply(
   A=StaticSwitchParameter(True=@n5, False=VectorParameter() {group=Blend, parameterName=BlendColor, defaultValue=<rgba 0.78 0.78 0.78 1>}.RGB) {group=Options, parameterName=UseGrassColorforBlendTexture}
   B=TextureSampleParameter2D(UVs=@n9) {group=Blend, parameterName=BlendTexture, texture=/Game/Environments/Homebase/Textures/T_HB_Grass_D}.RGB
  )
  n12 = ScalarParameter() {defaultValue=1, group=Blend2, parameterName=BlendMaskScale}
  n13 = Add(
   A=TextureSample(UVs=Multiply(A=@n2, B=@n12)) {samplerType=SAMPLERTYPE_LinearColor, texture=/Game/Environments/Homebase/Textures/T_HB_Soil_Mask}.R
   B=ScalarParameter() {defaultValue=1, group=Blend2, parameterName=BlendMaskintensity}
  )
  n14 = LinearInterpolate(
   A=ScalarParameter() {defaultValue=1, group=Blend2, parameterName=BlendMask2LOW}
   B=ScalarParameter() {defaultValue=1, group=Blend2, parameterName=BlendMask2High}
   Alpha=Multiply(A=VertexColor().B, B=Clamp(None=@n13))
  )
  n15 = Clamp(None=@n14)
  n16 = LinearInterpolate(
   A=Multiply(A=@n8, B=TextureSampleParameter2D(UVs=@n10) {parameterName=BlendTexture2, texture=/Game/Environments/Homebase/Textures/T_HB_Soil_D}.RGB)
   B=@n11
   Alpha=@n15
  )
  n17 = ScalarParameter() {group=Blend, parameterName=BlendContrastLOW}
  n18 = ScalarParameter() {defaultValue=1, group=Blend, parameterName=BlendContrastHigh}
  n19 = Clamp(None=DotProduct(A=PixelNormalWS(), B=Constant3Vector() {constant=<rgba 0 0 1 0>}))
  n20 = Multiply(A=Clamp(None=LinearInterpolate(A=@n17, B=@n18, Alpha=@n19)), B=VertexColor().R)
  n21 = ScalarParameter() {defaultValue=1, parameterName=Blend2Scale}
  n22 = Multiply(
   A=TextureSampleParameter2D(UVs=Multiply(A=@n21, B=@n3)) {parameterName=BlendTexture2, texture=&T_HB_Soil_D}.RGB
   B=@n8
  )
  n23 = StaticSwitchParameter(
   True=LinearInterpolate(A=@n7, B=StaticSwitchParameter(True=@n16, False=@n11) {group=Options, parameterName=Blend2}, Alpha=@n20)
   False=StaticSwitchParameter(True=LinearInterpolate(A=@n22, B=@n7, Alpha=@n15), False=@n7) {group=Options, parameterName=Blend2}
  ) {group=Options, parameterName=Blend}
  n24 = VectorParameter() {group=SnowCap, parameterName=SnowCapColor, defaultValue=<rgba 0.885 0.885 0.885 1>}
  n25 = Subtract(A=WorldPosition().XYZ, B=ObjectPositionWS())
  n26 = Divide(
   A=Add(A=ComponentMask(None=@n25) {b=true}, B=ScalarParameter() {defaultValue=100, group=SnowCap, parameterName=SnowCapFadeOffset})
   B=ScalarParameter() {defaultValue=100, group=SnowCap, parameterName=SnowCapFadeDistance}
  )
  n27 = LinearInterpolate(
   A=ScalarParameter() {group=SnowCap, parameterName=SnowCapMaskLow}
   B=ScalarParameter() {defaultValue=1, group=SnowCap, parameterName=SnowCapMaskHigh}
   Alpha=DotProduct(A=PixelNormalWS(), B=Constant3Vector() {constant=<rgba 0 0 1 0>})
  )
  n28 = Clamp(None=Multiply(A=Clamp(None=@n26), B=Clamp(None=@n27)))
  LinearInterpolate(
   A=Constant3Vector() {constant=<rgba 0.5868 0.8892 1 1>}
   B=StaticSwitchParameter(True=LinearInterpolate(A=@n23, B=@n24.RGB, Alpha=@n28), False=@n23) {group=Options, parameterName=UseSnowCaps}
   Alpha=ScalarParameter() {parameterName=BlueTint}
  )
  MaterialFunctionCall() {materialFunction=/Game/Environments/Homebase/MaterialFunctions/MF_HB_SmallWorld}
  n29 = TextureCoordinate() {coordinateIndex=2}
  n30 = Transform(None=Constant3Vector() {constant=<rgba 0 0 1 0>}) {transformSourceType=TRANSFORMSOURCE_World, transformType=TRANSFORM_Tangent}
  n31 = TextureSampleParameter2D() {parameterName=Normal, samplerType=SAMPLERTYPE_Normal, texture=/Game/Environments/Homebase/Textures/T_HB_Rock_Scattered_N}
  MaterialFunctionCall(
   BaseNormal=AppendVector(A=TextureCoordinate() {coordinateIndex=1}, B=ComponentMask(None=@n29) {r=true})
   AdditionalNormal=StaticSwitchParameter(
    True=StaticSwitchParameter(True=@n30, False=@n31.RGB) {group=Options, parameterName=UseNormalZupforced}
    False=Constant3Vector() {constant=<rgba 0 0 1 0>}
   ) {defaultValue=true, group=Options, parameterName=UseNormal}
  ) {materialFunction=&BlendAngleCorrectedNormals}
  Saturate(None=@n14)
  n32 = ScalarParameter() {defaultValue=0.5, group=Base, parameterName=Roughness}
  StaticSwitchParameter(True=LinearInterpolate(A=@n32, B=ScalarParameter() {defaultValue=0.5, group=Blend, parameterName=RoughnessBlend}, Alpha=@n20), False=@n32) {group=Options, parameterName=Blend}
 ]
}
