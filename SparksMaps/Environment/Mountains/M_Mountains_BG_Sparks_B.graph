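# Nanite-enabled opaque material for Sparks background mountains that blends tinted rock with world-aligned soil and grass textures.
# Upward-facing pixel normals and vertex red form the slope blend, while vertex blue and a soil mask drive a second mask.
# Vertex green picks live or dead grass tint from the Homebase parameter collection, which also supplies rock color.
# A base normal packed in UV1 and UV2 combines with the detail normal map.
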Material /SparksMaps/Environment/Mountains/M_Mountains_BG_Sparks_B
graph {
 properties {bUsedWithNanite=true, numCustomizedUVs=3, pixelDepthOffsetMode=PDOM_Legacy, usedShadingModels=Default Lit}
 nodes [
  Constant() {r=0.2}
  LinearInterpolate(
   A=ScalarParameter() {defaultValue=0.5, group=Base, parameterName=Roughness}
   B=ScalarParameter() {defaultValue=0.5, group=Blend, parameterName=RoughnessBlend}
   Alpha=NamedRerouteUsage()
  )
  MaterialFunctionCall() {materialFunction=/Game/Environments/Homebase/MaterialFunctions/MF_HB_SmallWorld}
  n1 = TextureCoordinate() {coordinateIndex=2}
  MaterialFunctionCall(
   BaseNormal=AppendVector(A=TextureCoordinate() {coordinateIndex=1}, B=ComponentMask(None=@n1) {r=true})
   AdditionalNormal=TextureSampleParameter2D() {parameterName=Normal, samplerType=SAMPLERTYPE_Normal, texture=/Game/Environments/Homebase/Textures/T_HB_Rock_Scattered_N}.RGB
  ) {materialFunction=&BlendAngleCorrectedNormals}
  n2 = CollectionParameter() {collection=/Game/Environments/Homebase/Materials/HomeBaseParameterCollection, parameterName=GrassColorDead}
  n3 = CollectionParameter() {collection=&HomeBaseParameterCollection, parameterName=GrassColor}
  NamedRerouteDeclaration(None=LinearInterpolate(A=ComponentMask(None=@n2) {b=true, g=true, r=true}, B=ComponentMask(None=@n3) {b=true, g=true, r=true}, Alpha=VertexColor().G)) {name=GrassColor, nodeColor=<rgba 0 0.9766 1 1>}
  n4 = WorldPosition() {worldPositionShaderOffset=WPT_ExcludeAllShaderOffsets}
  n5 = CollectionParameter() {collection=&HomeBaseParameterCollection, parameterName=GrassTextureSize}
  NamedRerouteDeclaration(None=Divide(A=ComponentMask(None=@n4.XYZ) {g=true, r=true}, B=ComponentMask(None=@n5) {r=true})) {name=WorldAlignedUVs, nodeColor=<rgba 0.3828 0 1 1>}
  n6 = ScalarParameter() {defaultValue=1, group=Blend2, parameterName=BlendMask2LOW}
  n7 = ScalarParameter() {defaultValue=1, group=Blend2, parameterName=BlendMask2High}
  n8 = TextureSample(UVs=Multiply(A=NamedRerouteUsage(), B=ScalarParameter() {defaultValue=1, group=Blend2, parameterName=BlendMaskScale})) {samplerType=SAMPLERTYPE_LinearColor, texture=/Game/Environments/Homebase/Textures/T_HB_Soil_Mask}
  n9 = ScalarParameter() {defaultValue=1, group=Blend2, parameterName=BlendMaskintensity}
  n10 = Multiply(A=VertexColor().B, B=Saturate(None=Add(A=@n8.R, B=@n9)))
  NamedRerouteDeclaration(None=Saturate(None=LinearInterpolate(A=@n6, B=@n7, Alpha=@n10))) {name=BlendMask, nodeColor=<rgba 1 0.1406 0 1>}
  n11 = CollectionParameter() {collection=&HomeBaseParameterCollection, parameterName=RockColor}
  NamedRerouteDeclaration(None=Multiply(
    A=TextureSampleParameter2D() {parameterName=Diffuse, texture=/Game/Environments/Homebase/Textures/T_HB_Rock_Scattered_D}.RGB
    B=ComponentMask(None=@n11) {b=true, g=true, r=true}
   )) {name=RockColor, nodeColor=<rgba 0 1 0.8594 1>}
  n12 = PixelNormalWS()
  n13 = Constant3Vector() {constant=<rgba 0 0 1 0>}
  n14 = LinearInterpolate(
   A=ScalarParameter() {group=Blend, parameterName=BlendContrastLOW}
   B=ScalarParameter() {defaultValue=1, group=Blend, parameterName=BlendContrastHigh}
   Alpha=Saturate(None=DotProduct(A=@n12, B=@n13))
  )
  NamedRerouteDeclaration(None=Multiply(A=Saturate(None=@n14), B=VertexColor().R)) {name=ZBlend, nodeColor=<rgba 1 0 0.5859 1>}
  n15 = TextureSampleParameter2D(UVs=Multiply(A=ScalarParameter() {defaultValue=1, parameterName=Blend2Scale}, B=NamedRerouteUsage())) {parameterName=BlendTexture2, texture=/Game/Environments/Homebase/Textures/T_HB_Soil_D}
  n16 = Multiply(
   A=NamedRerouteUsage()
   B=TextureSampleParameter2D(UVs=NamedRerouteUsage()) {group=Blend, parameterName=BlendTexture, texture=/Game/Environments/Homebase/Textures/T_HB_Grass_D}.RGB
  )
  n17 = NamedRerouteUsage()
  NamedRerouteDeclaration(None=LinearInterpolate(A=NamedRerouteUsage(), B=LinearInterpolate(A=@n15.RGB, B=@n16, Alpha=@n17), Alpha=NamedRerouteUsage())) {name=TextureBlend, nodeColor=<rgba 0 0.9062 1 1>}
  NamedRerouteUsage()
 ]
}
